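# Sound cue for the Quad rocket launcher's firing sound, layered by listener distance.
# A mixer plays close, medium and distant layers together, each randomly picking a wave variant through its own modulator and distance attenuation.
# The close layer uses a third-person ranged-fire sound class, while medium and distant share a distant-fire class.
# Local and global gunfire concurrency sets are assigned to the cue.
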SoundCue /Game/Sounds/Fort_Weapons/Launchers/Quad/Weapon_LAU_Quad_Fire_Cue
root=Mixer_3
nodes [
 {
  id=Attenuation_10
  children [Modulator_0]
  properties {attenuationSettings=&Fort_Weapon_Projectile_Medium_Mono_Athena_Att}
 }
 {
  id=Attenuation_11
  children [Modulator_7]
  properties {attenuationSettings=&Fort_Weapon_Projectile_Distant_Mono_Athena_Att}
 }
 {id=Attenuation_9, children [Modulator_6], properties {attenuationSettings=&Fort_Weapon_Projectile_Close_Athena_Att}}
 {id=Mixer_3, children [SoundClass_7,SoundClass_4,SoundClass_3], properties {inputVolume [1,1,1]}}
 {id=Modulator_0, children [Random_3], properties {volumeMax=2.05, volumeMin=1.95}}
 {id=Modulator_6, children [Random_0], properties {pitchMax=1, pitchMin=1}}
 {id=Modulator_7, children [Random_4], properties {pitchMax=1.1, volumeMax=2.05, volumeMin=1.95}}
 {id=Random_0, children [WavePlayer_1,WavePlayer_2,WavePlayer_3,WavePlayer_4], properties {weights [1,1,1,1]}}
 {id=Random_3, children [WavePlayer_0, WavePlayer_12, WavePlayer_13, WavePlayer_14], properties {weights [1,1,1,1]}}
 {id=Random_4, children [WavePlayer_15, WavePlayer_16, WavePlayer_17], properties {weights [1,1,1]}}
 {
  id=SoundClass_3
  children [Attenuation_11]
  properties {soundClassOverride=&Fort_WeaponRanged_Fire_Distant}
 }
 {
  id=SoundClass_4
  children [Attenuation_10]
  properties {soundClassOverride=&Fort_WeaponRanged_Fire_Distant}
 }
 {
  id=SoundClass_7
  children [Attenuation_9]
  properties {soundClassOverride=&Fort_WeaponRanged_Fire_3P_PSM}
 }
 {id=WavePlayer_0, properties {soundWaveAssetPtr=&/QuadRocket_Fire_MediumNew_01}}
 {id=WavePlayer_1, properties {soundWaveAssetPtr=&/QuadRocket_Fire_Close_01}}
 {id=WavePlayer_12, properties {soundWaveAssetPtr=&/QuadRocket_Fire_MediumNew_03}}
 {id=WavePlayer_13, properties {soundWaveAssetPtr=&/QuadRocket_Fire_MediumNew_02}}
 {id=WavePlayer_14, properties {soundWaveAssetPtr=&/QuadRocket_Fire_MediumNew_05}}
 {id=WavePlayer_15, properties {soundWaveAssetPtr=&/QuadRocket_Fire_DistantNew_02}}
 {id=WavePlayer_16, properties {soundWaveAssetPtr=&/QuadRocket_Fire_DistantNew_04}}
 {id=WavePlayer_17, properties {soundWaveAssetPtr=&/QuadRocket_Fire_DistantNew_05}}
 {id=WavePlayer_2, properties {soundWaveAssetPtr=&/QuadRocket_Fire_Close_03}}
 {id=WavePlayer_3, properties {soundWaveAssetPtr=&/QuadRocket_Fire_Close_04}}
 {id=WavePlayer_4, properties {soundWaveAssetPtr=&/QuadRocket_Fire_Close_06}}
]
properties {
 maxDistance=27500
 volumeMultiplier=0.5
 attenuationOverrides {falloffDistance=5000, attenuationShapeExtents {x=1000}}
 concurrencyOverrides {resolutionRule=PreventNew}
 concurrencySet [&Weapon_Gunfire_Local_Conc, &Weapon_Gunfire_Global_Conc]
}
