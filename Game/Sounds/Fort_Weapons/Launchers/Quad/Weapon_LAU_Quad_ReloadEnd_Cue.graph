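# Sound cue for the end-of-reload sound of the Quad launcher weapon, in the weapon foley sound class.
# A single wave player feeds the QuadRocket_ReloadEnd_02 wave through a Modulator at the root, with no random choice between samples.
# The shared weapon reload attenuation asset spatializes it, and a local concurrency override allows one instance at a time, stopping the farthest and then preventing new ones.
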SoundCue /Game/Sounds/Fort_Weapons/Launchers/Quad/Weapon_LAU_Quad_ReloadEnd_Cue
root=Modulator_1
nodes [
 {id=Modulator_1, children [WavePlayer_2]}
 {id=WavePlayer_2, properties {soundWaveAssetPtr=&/QuadRocket_ReloadEnd_02}}
]
properties {
 attenuationSettings=&Fort_Weapon_Reload_Att
 bOverrideConcurrency=true
 maxDistance=2250
 soundClassObject=&Fort_WeaponFoley
 volumeMultiplier=0.902
 attenuationOverrides {falloffDistance=5000, attenuationShapeExtents {x=1000}}
 concurrencyOverrides {maxCount=1, resolutionRule=StopFarthestThenPreventNew}
}
