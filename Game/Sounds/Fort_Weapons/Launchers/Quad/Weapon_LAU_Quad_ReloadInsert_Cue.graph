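# Sound cue for the Quad launcher's rocket insert during reload, routed to the weapon foley class with reload attenuation.
# A mixer layers a random pick of two insert waves, passed through a modulator for variation, with a delayed random pick of two slam waves that lands a fixed beat later.
# A concurrency override allows one instance at a time, stopping the farthest before preventing new ones.
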SoundCue /Game/Sounds/Fort_Weapons/Launchers/Quad/Weapon_LAU_Quad_ReloadInsert_Cue
root=Mixer_0
nodes [
 {id=Delay_0, children [Random_2], properties {delayMax=0.7, delayMin=0.7}}
 {id=Mixer_0, children [Modulator_1,Delay_0], properties {inputVolume [1,1]}}
 {id=Modulator_1, children [Random_0]}
 {id=Random_0, children [WavePlayer_1,WavePlayer_2], properties {weights [1,1]}}
 {id=Random_2, children [WavePlayer_5,WavePlayer_6], properties {weights [1,1]}}
 {id=WavePlayer_1, properties {soundWaveAssetPtr=&/QuadRocket_ReloadInsert__01}}
 {id=WavePlayer_2, properties {soundWaveAssetPtr=&/QuadRocket_ReloadInsert__02}}
 {id=WavePlayer_5, properties {soundWaveAssetPtr=&/QuadRocket_ReloadInsertSlam_01}}
 {id=WavePlayer_6, properties {soundWaveAssetPtr=&/QuadRocket_ReloadInsertSlam_02}}
]
properties {
 attenuationSettings=&Fort_Weapon_Reload_Att
 bOverrideConcurrency=true
 maxDistance=2250
 soundClassObject=&Fort_WeaponFoley
 volumeMultiplier=0.702
 attenuationOverrides {falloffDistance=5000, attenuationShapeExtents {x=1000}}
 concurrencyOverrides {maxCount=1, resolutionRule=StopFarthestThenPreventNew}
}
