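# Sound cue for the start of the quad rocket launcher reload, played as weapon foley with reload attenuation.
# A mixer layers three randomized parts: a modulated reload-start sound, an eject sound after a short fixed delay, and a quieter eject-impact sound after a longer one.
# Concurrency is overridden to a single instance, stopping the farthest before preventing new ones.
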SoundCue /Game/Sounds/Fort_Weapons/Launchers/Quad/Weapon_LAU_Quad_ReloadStart_Cue
root=Mixer_0
nodes [
 {id=Delay_0, children [Random_1], properties {delayMax=0.5, delayMin=0.5}}
 {id=Delay_1, children [Random_3], properties {delayMax=1, delayMin=1}}
 {id=Mixer_0, children [Modulator_1,Delay_0,Delay_1], properties {inputVolume [1,1,0.3]}}
 {id=Modulator_1, children [Random_0]}
 {id=Random_0, children [WavePlayer_1,WavePlayer_2], properties {weights [1,1]}}
 {id=Random_1, children [WavePlayer_3,WavePlayer_4], properties {weights [1,1]}}
 {id=Random_3, children [WavePlayer_5,WavePlayer_6,WavePlayer_7], properties {weights [1,1,1]}}
 {id=WavePlayer_1, properties {soundWaveAssetPtr=&/QuadRocket_ReloadStart_01}}
 {id=WavePlayer_2, properties {soundWaveAssetPtr=&/QuadRocket_ReloadStart_02}}
 {id=WavePlayer_3, properties {soundWaveAssetPtr=&/QuadRocket_ReloadEject_01}}
 {id=WavePlayer_4, properties {soundWaveAssetPtr=&/QuadRocket_ReloadEject_02}}
 {id=WavePlayer_5, properties {soundWaveAssetPtr=&/QuadRocket_ReloadEjectImpact_01}}
 {id=WavePlayer_6, properties {soundWaveAssetPtr=&/QuadRocket_ReloadEjectImpact_03}}
 {id=WavePlayer_7, properties {soundWaveAssetPtr=&/QuadRocket_ReloadEjectImpact_02}}
]
properties {
 attenuationSettings=&Fort_Weapon_Reload_Att
 bOverrideConcurrency=true
 maxDistance=2250
 soundClassObject=&Fort_WeaponFoley
 volumeMultiplier=0.702
 attenuationOverrides {falloffDistance=5000, attenuationShapeExtents {x=1000}}
 concurrencyOverrides {maxCount=1, resolutionRule=StopFarthestThenPreventNew}
}
