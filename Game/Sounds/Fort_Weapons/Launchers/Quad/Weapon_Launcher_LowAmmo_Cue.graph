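# Sound cue for the low-ammo warning on the quad launcher weapon.
# A single wave player plays the quad launcher low-ammo wave directly as the cue output, with no randomization or mixing.
# It routes to a ranged-weapon low-ammo sound class and uses close-range Athena projectile attenuation.
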SoundCue /Game/Sounds/Fort_Weapons/Launchers/Quad/Weapon_Launcher_LowAmmo_Cue
root=WavePlayer_0
nodes [{id=WavePlayer_0, properties {soundWaveAssetPtr=&/Weapon_LauncherQuad_LowAmmo}}]
properties {
 attenuationSettings=&Fort_Weapon_Projectile_Close_Athena_Att
 maxDistance=4950
 soundClassObject=/Game/Sounds/FortSoundClassesAndModes/Fort_WeaponRanged_LowAmmo_Degraded_PSM
}
